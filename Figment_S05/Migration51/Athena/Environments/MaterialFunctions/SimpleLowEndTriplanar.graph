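# Material function for a cheap triplanar-style world-space projection of an input base color texture, using a single texture sample.
# World position divided by texture size supplies UVs, hard-switched between XZ and YZ by the dominant horizontal vertex normal, with no top-down XY projection.
# Explicit mip derivatives come from scaled mesh UVs, avoiding derivatives of the switched coordinates.
# The output appends the sampled RGB to itself.
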MaterialFunction /Figment_S05/Migration51/Athena/Environments/MaterialFunctions/SimpleLowEndTriplanar
graph {nodes [
  n1 = FunctionInput() {
   bUsePreviewValueAsDefault=true
   description=Size of texture in worldspace units
   inputName=TextureSize
   sortPriority=3
   previewValue=<64 64 64 6>
  }
  n2 = Divide(A=FunctionInput(Preview=WorldPosition().XYZ) {inputName=Worldpos, sortPriority=1}, B=Multiply(A=Abs(None=@n1)) {constB=-1})
  n3 = Abs(None=VertexNormalWS())
  n4 = Ceil(None=Subtract(A=ComponentMask(None=@n3) {r=true}, B=ComponentMask(None=@n3) {g=true}))
  n5 = Multiply(A=TextureCoordinate(), B=ScalarParameter() {defaultValue=0.25, parameterName=MipDerivativeScale})
  n6 = TextureSample!(
   UVs=LinearInterpolate(A=ComponentMask(None=@n2) {b=true, r=true}, B=ComponentMask(None=@n2) {b=true, g=true}, Alpha=Saturate(None=@n4))
   Tex=FunctionInput() {inputName=Basecolor Texture, inputType=FunctionInput_Texture2D}
   "DDX(UVs)"=DDX(Value=@n5)
   "DDY(UVs)"=DDY(Value=@n5)
   Apply View MipBias
  ) {desc=Y projection, mipValueMode=TMVM_Derivative, samplerSource=SSM_Wrap_WorldGroupSettings}
  FunctionOutput(None=AppendVector(A=@n6.RGB, B=@n6.RGB))
 ]}
